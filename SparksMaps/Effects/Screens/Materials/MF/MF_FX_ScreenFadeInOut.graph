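# Material function for screen effects that fades a Base Color in from, or out to, a FadeColor (black by default).
# A custom primitive data vector packs FadeInOut, CurrentTime, FadeSpeed and SkipFade.
# FadeInOut selects a rising or falling ramp from material Time offset by CurrentTime, scaled by FadeSpeed and clamped.
# The ramp blends the two colors; despite its name, SkipFade at 0 bypasses the fade and 1 applies it.
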MaterialFunction /SparksMaps/Effects/Screens/Materials/MF/MF_FX_ScreenFadeInOut
graph {nodes [
  n1 = Constant3Vector()
  n2 = Constant3Vector() {constant=<rgba 0.205 1 0.2263 1>}
  n3 = Time()
  n4 = VectorParameter() {
   bUseCustomPrimitiveData=true
   parameterName=ScreenFadePrimitiveData
   channelNames {a=SkipFade, b=FadeSpeed, g=CurrentTime, r=FadeInOut}
   defaultValue=<rgba 1 0 1 1>
  }
  n5 = LinearInterpolate(A=OneMinus(None=@n3), B=@n3, Alpha=@n4.FadeInOut)
  n6 = Multiply(A=@n4.CurrentTime, B=LinearInterpolate(Alpha=@n4.FadeInOut) {constA=-1})
  n7 = Saturate(None=Multiply(A=Subtract(A=@n5, B=@n6), B=@n4.FadeSpeed) {constB=0.5})
  n8 = Saturate(None=@n4.SkipFade)
  FunctionOutput(None=LinearInterpolate(
    A=FunctionInput(Preview=@n1) {bUsePreviewValueAsDefault=true, description=Which Color it Will fade in or out to, inputName=FadeColor}
    B=FunctionInput(Preview=@n2) {description=The color that will be replaced by the fade., inputName=Base Color}
    Alpha=LinearInterpolate(B=@n7, Alpha=@n8) {constA=1, constAlpha=0}
   ) {constAlpha=0})
 ]}
