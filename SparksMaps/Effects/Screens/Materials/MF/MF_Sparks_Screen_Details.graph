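# Material function for Sparks stage screens that adds LCD or scanline detail to an Emissive input and scales its brightness.
# A static switch picks LCD mask or camera-faded sine scanlines, multiplied by a custom-primitive-data screen fade.
# An MPC GI setting scales emissive before eye adaptation; ray-traced quality boosts it toward the stage primary color, while low/mobile uses a fixed multiplier.
# Pixel pattern, flicker and projected-UV branches never reach the output.
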MaterialFunction /SparksMaps/Effects/Screens/Materials/MF/MF_Sparks_Screen_Details
graph {nodes [
  n1 = StaticSwitchParameter(
   True=ScalarParameter() {bUseCustomPrimitiveData=true, defaultValue=200, parameterName=LCD Size CPD, primitiveDataIndex=1}
   False=ScalarParameter() {defaultValue=200, group=Screen Details, parameterName=LCD Size}
  ) {group=Screen Details, parameterName=UseLCDTileCPD}
  n2 = TextureCoordinate()
  n3 = Add(
   A=ScalarParameter() {defaultValue=0.5, group=Screen Details, parameterName=LCD Brightness}
   B=TextureSample(UVs=Multiply(A=@n1, B=@n2)) {samplerType=SAMPLERTYPE_Masks, texture=/Game/Environments/Asteria/Sets/NeonCity/Techart/Textures/T_LCD_Panels_LCDSoft}.RGB
  )
  n4 = ComponentMask(None=TextureCoordinate()) {g=true}
  n5 = FunctionInput(Preview=StaticSwitchParameter(
    True=ScalarParameter() {bUseCustomPrimitiveData=true, group=Screen Details, parameterName=Scanlines Tile CPD}
    False=ScalarParameter() {defaultValue=100, group=Screen Details, parameterName=Scanlines Tile}
   ) {group=Screen Details, parameterName=UseScanlinesTileCPD}) {bUsePreviewValueAsDefault=true, inputName=Scanlines Tile, inputType=FunctionInput_Scalar, sortPriority=1}
  n6 = ConstantBiasScale(None=Sine(None=Multiply(A=@n4, B=@n5) {constB=125}))
  n7 = OneMinus(None=ScalarParameter() {defaultValue=0.25, group=Screen Details, parameterName=Scanelines Intensity})
  n8 = LinearInterpolate(
   A=Saturate(None=Add(A=@n6, B=@n7) {constB=0.25})
   Alpha=MaterialFunctionCall(Fade Length=ScalarParameter() {defaultValue=15000, group=Screen Details, parameterName=Scanlines Camera Fade}) {materialFunction=&CameraDepthFade}.Result
  ) {constA=1}
  n9 = Multiply(
   A=StaticSwitchParameter(True=Saturate(None=@n3), False=NamedRerouteDeclaration(None=@n8) {name=Scanlines_A, nodeColor=<rgba 0 0 0 1>}) {group=Screen Details, parameterName=UseLCD}
   B=MaterialFunctionCall!(FadeColor, Base Color=FunctionInput() {inputName=Emissive}) {desc=Uses CPD to fade screens in screen BP, materialFunction=&/MF_FX_ScreenFadeInOut}.Result
  )
  n10 = LinearInterpolate(
   A=ScalarParameter() {defaultValue=1.5, group=Emissive Adjustments, parameterName=LowGIQualityMultiplier}
   B=ScalarParameter() {defaultValue=10, group=Emissive Adjustments, parameterName=HighGIQualityMultiplier}
   Alpha=CollectionParameter() {collection=/SparksEffects/Effects/MPC_Sparks, desc=Set within screen manager, parameterName=ScreensHighGISetting}
  ) {constA=1.5, constB=10, desc="Value of 10 for Lumen/GI Quality high/epic, Value of 1.5 for GI quality set to OFF or Ambient Occlusion"}
  n11 = ScalarParameter() {defaultValue=1, group=Emissive Adjustments, parameterName=Cosmetic Emissive EyeAdaptation Alpha}
  n12 = MaterialFunctionCall!(Emissive=Multiply(A=@n9, B=@n10), Alpha=@n11, Material) {materialFunction=/Game/Packages/Fortress_SharedMaterials/Base_Material_Functions/BaseCharacterShaderFunctions/MF_CosmeticEmissiveEyeAdaptation}
  n13 = MaterialFunctionCall!(CustomUPosition) {materialFunction=/SparksStageShow/Systems/StageManager/FXManager/MaterialFunctions/MF_MojoStageColors}
  n14 = ScalarParameter() {defaultValue=1, group=Emissive Adjustments, parameterName=LumenUseColorOverride}
  n15 = RayTracingQualitySwitch(
   Normal=@n12.Emissive
   RayTraced=Multiply(
    A=ScalarParameter() {defaultValue=2, group=Emissive Adjustments, parameterName=LumenBoost}
    B=LinearInterpolate(A=@n12.Emissive, B=@n13.Primary, Alpha=@n14)
   )
  )
  n16 = MaterialFunctionCall(
   Emissive=Multiply(A=@n9, B=ScalarParameter() {defaultValue=1.5, group=Emissive Adjustments, parameterName=LowEndMultiplier}) {constB=1.5, desc=Low End and Mobile Screen Multiplier}
   Alpha=@n11
  ) {materialFunction=&MF_CosmeticEmissiveEyeAdaptation}
  FunctionOutput(None=ShadingPathSwitch(Default=QualitySwitch(Default=@n15, Low=@n16.Emissive), Mobile=@n16.Emissive))
  n17 = Constant3Vector() {constant=<rgba 64 64 64 0>}
  n18 = Divide(A=WorldPosition() {worldPositionShaderOffset=WPT_ExcludeAllShaderOffsets}.XYZ, B=Multiply(A=Abs(None=@n17)) {constB=-1})
  n19 = VertexNormalWS()
  n20 = Abs(None=ComponentMask(None=@n19) {b=true})
  n21 = Constant() {desc=projection contrast scale, r=1}
  n22 = Abs(None=ComponentMask(None=@n19) {r=true})
  n23 = LinearInterpolate(
   A=ComponentMask(None=@n18) {b=true, r=true}
   B=ComponentMask(None=@n18) {b=true, g=true}
   Alpha=ComponentMask(None=MaterialFunctionCall(In=@n22, Contrast=@n21) {materialFunction=&CheapContrast}.Result) {r=true}
  ) {desc=xy uv}
  LinearInterpolate(
   A=@n23
   B=ComponentMask(None=@n18) {g=true, r=true}
   Alpha=ComponentMask(None=MaterialFunctionCall(In=@n20, Contrast=@n21) {materialFunction=&CheapContrast}.Result) {r=true}
  ) {desc=xyz uv}
  n24 = MaterialFunctionCall(f2=Multiply(A=Abs()) {constB=500}) {materialFunction=&AddComponents}
  n25 = Multiply(A=TextureCoordinate(), B=Constant2Vector() {g=1250, r=1250})
  n26 = NamedRerouteUsage()
  NamedRerouteDeclaration(None=Multiply(
    A=TextureSample!(UVs=@n25, Tex, Level=@n24.f2, Apply View MipBias) {mipValueMode=TMVM_MipLevel, texture=/Game/Environments/Homebase/Textures/pixelPattern}.RGB
    B=LinearInterpolate(Alpha=@n26) {constB=0.1}
   )) {name=Pixels Color, nodeColor=<rgba 0 0.2734 1 1>}
  NamedRerouteDeclaration(None=@n23) {name=World Space UVs, nodeColor=<rgba 0 0.5547 1 1>}
  n27 = MaterialFunctionCall(Vector 3=Subtract(A=WorldPosition().XYZ, B=CameraPositionWS())) {materialFunction=/Engine/Functions/Engine_MaterialFunctions02/Utility/VectorLength}
  n28 = Constant() {r=150}
  n29 = Constant() {r=700}
  NamedRerouteDeclaration(None=OneMinus(None=MaterialFunctionCall(Input=@n27.V3 Length, Input Low=@n28, Input High=@n29) {materialFunction=&RemapValue_0-1}.Result)) {name=PixelBlendFade, nodeColor=<rgba 1 0 0.7969 1>}
  NamedRerouteDeclaration(None=MaterialFunctionCall!(Time Mult, Time Add) {materialFunction=/Game/Effects/Niagara/MasterMaterials/Functions/Mojo_MF/MF_FX_Flicker}.Raw Flicker) {name=Flicker, nodeColor=<rgba 1 0.5625 0 1>}
  Reroute(None=MaterialFunctionCall!(Position, UVRadialScale=Constant() {r=0.1}) {materialFunction=/Game/Weapons/WeaponSkins/MaterialFunction/ObjectStabilizedScreenSpaceUVs}.StabilizedUVs) {desc=text uvs}
  n30 = ObjectPositionWS()
  Reroute(None=Divide(
    A=MaterialFunctionCall!(Center Point World Position=@n30) {materialFunction=/Engine/Functions/Engine_MaterialFunctions02/UVs/WorldSpaceAlignedScreenCoordinates}.Offset UVs
    B=Constant() {r=4}
   ) {constB=5}) {desc=text uvs}
  n31 = Reroute()
  n32 = NamedRerouteUsage()
  n33 = Constant() {desc="Pixel Color On/Off (This could be a input control?)", r=0.2}
  StaticSwitchParameter(True=LinearInterpolate(A=@n31, B=Add(A=@n31, B=@n32), Alpha=Saturate(None=@n33)) {desc=Pixel Color Blend}, False=@n31) {group=Screen Details, parameterName=Use Pixel Details?}
 ]}
